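# Physics asset for the DEMO skeletal mesh of the Muertos pickaxe, a Fortnite melee weapon cosmetic.
# It holds a single capsule body on the Melee_TwoHanded_Handle bone, offset and rotated to serve as the collision shape.
# It has no constraints, so the pickaxe simulates as one rigid piece.
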PhysicsAsset /BRCosmetics/Weapons/FORT_Melee/Pickaxe_Muertos/Meshes/DEMO/SK_Pickaxe_Muertos_DEMO_PhysicsAsset
bodies {Melee_TwoHanded_Handle [{shapeType=Capsule, length=71.8656, radius=22.9896, center=<-5.8004 -0.8362 34.6844>, rotation=<rot -0.2111 -90 -0.1008>}]}
constraints []
